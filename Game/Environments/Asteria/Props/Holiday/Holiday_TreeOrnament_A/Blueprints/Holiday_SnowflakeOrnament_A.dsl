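; Holiday snowflake ornament Blueprint, a static mesh prop whose look is picked from a set of mesh variants.
; The construction script clamps MeshIndex to the valid range of the Meshes array, writes it back, and assigns that entry to the StaticMeshComponent before calling the parent script.
; BeginPlay, overlap and Tick are disabled stubs, so it has no runtime logic of its own.
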(event EventBeginPlay
  (|CallParentFunction) ; This node is disabled and will not be called.\nDrag off pins to build functionality.
)

(event Collision|EventActorBeginOverlap
  (params (OtherActor :type Actor%20Object%20Reference))
  (|CallParentFunction OtherActor) ; This node is disabled and will not be called.\nDrag off pins to build functionality.
)

(event EventTick
  (params (DeltaSeconds :type Float%20%28single-precision%29))
  (|CallParentFunction DeltaSeconds) ; This node is disabled and will not be called.\nDrag off pins to build functionality.
)

(fn ConstructionScript
  (bind _output_get (Variables|HolidayDecoration|SetMeshIndex (Math|Integer|Clamp(Integer) (Variables|HolidayDecoration|GetMeshIndex) 0 (Utilities|Array|LastIndex (Variables|HolidayDecoration|GetMeshes)))))
  (Components|StaticMesh|SetStaticMesh (Variables|StaticMeshActor|GetStaticMeshComponent) (Utilities|Array|Get(acopy) (Variables|HolidayDecoration|GetMeshes) _output_get))
  (|CallParentFunction))
