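# Material for a hanging holiday tree ornament with a glowing metallic body and a separately colored ribbon, selected by UV tile.
# World position offset twists it back and forth around the pivot's Z axis, fading by bounding-box height, with phase and speed varied by pivot location.
# Normals follow the rotation and feed an axis-aligned Fresnel that shapes the emissive.
# Glow has mobile and Switch tuning, plus a flat multiplier for ray-traced lighting.
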Material /Game/Environments/Asteria/Props/Holiday/Holiday_TreeOrnament_A/Materials/M_Holiday_HangingOrnament
graph {
 properties {bUseMaterialAttributes=true, bUsedWithNanite=true, pixelDepthOffsetMode=PDOM_Legacy, twoSided=true, usedShadingModels=Default Lit}
 nodes [
  n1 = MaterialFunctionCall() {materialFunction=&BoundingBoxBased_0-1_UVW}
  n2 = ScalarParameter() {defaultValue=0.25, parameterName=TwistMaskPercent, sliderMax=1}
  n3 = Saturate(None=Divide(A=Subtract(A=@n1.B, B=@n2) {constB=0.2}, B=Add(A=@n2)))
  Ceil(None=@n3) {desc=Use to preview mesh mask}
  n4 = Constant3Vector() {constant=<rgba 0 1 0 0>}
  n5 = Add(
   A=@n4
   B=RotateAboutAxis(
    NormalizedRotationAxis=Constant3Vector() {constant=<rgba 0 0 1 0>}
    RotationAngle=NamedRerouteUsage()
    PivotPoint=Constant3Vector()
    Position=@n4
   )
  )
  n6 = ScalarParameter() {defaultValue=10, parameterName=GlowFresnelExponent}
  n7 = Multiply(A=@n6, B=ScalarParameter() {defaultValue=0.3, parameterName=GlowFresnelExponentScale_Mobile})
  n8 = QualitySwitch(Default=@n7, Low=@n7, Medium=Multiply(A=@n6, B=ScalarParameter() {defaultValue=0.33, parameterName=GlowFresnelExponentScale_Switch}))
  n9 = RayTracingQualitySwitch(
   Normal=MaterialFunctionCall!(Axis=Transform(None=@n5), Exponent=ShadingPathSwitch(Default=@n6, Deferred=@n6, Mobile=@n8), Normal=NamedRerouteUsage()) {materialFunction=/Game/Effects/Fort_Effects/Materials/_Functions/MF_AxisAlignedFrensnel}.Fresnel
   RayTraced=ScalarParameter() {defaultValue=0.1, parameterName=LumenMult}
  )
  n10 = ScalarParameter() {defaultValue=1, parameterName=LightsOn, sliderMax=1}
  n11 = VectorParameter() {parameterName=Color, defaultValue=<rgba 0.0043 1 0 1>}
  n12 = Multiply(A=@n11.RGB, B=ScalarParameter() {defaultValue=8, parameterName=Brightness})
  n13 = Multiply(A=@n11.RGB, B=ScalarParameter() {defaultValue=3, parameterName=Brightness_Mobile})
  n14 = ScalarParameter() {defaultValue=3, parameterName=Brightness_Switch}
  n15 = ShadingPathSwitch(Default=@n12, Deferred=@n12, Mobile=QualitySwitch(Default=@n13, Low=@n13, Medium=Multiply(A=@n11.RGB, B=@n14)))
  n16 = Floor(None=ComponentMask(None=TextureCoordinate()) {r=true})
  n17 = OneMinus(None=NamedRerouteDeclaration(None=Saturate(None=@n16)) {name=RibbonMask, nodeColor=<rgba 0 1 0.6719 1>})
  n18 = SetMaterialAttributes!(
   MaterialAttributes
   Base Color=Constant3Vector() {constant=<rgba 1 1 1 0>}
   Metallic=ScalarParameter() {defaultValue=1, parameterName=Metallic, sliderMax=1}
   Roughness=ScalarParameter() {defaultValue=0.148, parameterName=Roughness, sliderMax=1}
   Emissive Color=Multiply(A=Multiply(A=@n9, B=@n10), B=Multiply(A=@n15, B=@n17))
  ) {attributeSetTypes [
    #1
    #2
    #3
    #4
   ]}
  n19 = VectorParameter() {parameterName=RibbonColor, defaultValue=<rgba 0.6979 0.0145 0.0145 0>}
  n20 = ScalarParameter() {defaultValue=0.68, parameterName=RibbonMetallic, sliderMax=1}
  n21 = ScalarParameter() {defaultValue=0.414, parameterName=RibbonRoughness, sliderMax=1}
  n22 = Constant3Vector() {constant=<rgba 0 0 1 0>}
  n23 = OneMinus(None=@n3)
  n24 = ScalarParameter() {defaultValue=0.07, parameterName=RotateAmt}
  n25 = MaterialFunctionCall() {materialFunction=&ObjectPivotPoint}
  n26 = Frac(None=Divide(A=MaterialFunctionCall(f3=@n25.Object Pivot Location) {materialFunction=&AddComponents}.f3) {constB=100})
  n27 = Multiply(
   A=Add(A=Time(), B=@n26)
   B=LinearInterpolate(
    A=ScalarParameter() {defaultValue=0.04, parameterName=RotateSpeed_Min}
    B=ScalarParameter() {defaultValue=0.2, parameterName=RotateSpeed_Max}
    Alpha=@n26
   ) {constA=0.1, constB=0.3}
  )
  n28 = NamedRerouteDeclaration(None=Multiply(A=Multiply(A=@n23, B=@n24), B=Sine(None=@n27))) {name=RotationAngle, nodeColor=<rgba 1 0.5859 0 1>}
  n29 = VertexNormalWS()
  n30 = RotateAboutAxis(NormalizedRotationAxis=@n22, RotationAngle=@n28, PivotPoint=Constant3Vector(), Position=@n29)
  SetMaterialAttributes!(
   MaterialAttributes=BlendMaterialAttributes(
    A=@n18
    B=SetMaterialAttributes!(MaterialAttributes, Base Color=@n19.RGB, Metallic=@n20, Roughness=@n21) {attributeSetTypes [#1, #2, #3]}
    Alpha=NamedRerouteUsage()
   )
   World Position Offset=RotateAboutAxis(
    NormalizedRotationAxis=@n22
    RotationAngle=@n28
    PivotPoint=MaterialFunctionCall() {materialFunction=&ObjectPivotPoint}.Object Pivot Location
    Position=WorldPosition().XYZ
   )
   Normal=NamedRerouteDeclaration(None=Add(A=@n29, B=@n30)) {name=RotatedNormal, nodeColor=<rgba 0 1 0.5313 1>}
  ) {attributeSetTypes [#5, #6]}
 ]
}
